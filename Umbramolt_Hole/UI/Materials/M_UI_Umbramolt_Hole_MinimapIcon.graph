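# Translucent UI material for the Umbramolt Hole minimap icon, layering the Umbramolt Kipper texture with animated glow effects.
# A center-scaled icon sample combines with dimmed masks through Max for alpha, while WaveColor and PulseColor blend for tint.
# Named reroutes define Wave, time-scrolled cosine bands over a four-sided SDF, and Pulse, animated plus static circle glows.
# Separate hexagon SDF nodes build an expanding, fading outline that feeds neither reroute.
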Material /Umbramolt_Hole/UI/Materials/M_UI_Umbramolt_Hole_MinimapIcon
graph {
 properties {bAutomaticallySetUsageInEditor=false, blendMode=BLEND_Translucent, materialDomain=MD_UI, usedShadingModels=Default Lit}
 nodes [
  Constant()
  n1 = TextureSampleParameter2D(UVs=MaterialFunctionCall(UVs=TextureCoordinate(), Texture Scale=ScalarParameter() {defaultValue=0.25, group=Icon, parameterName=IconScale}) {materialFunction=&ScaleUVsByCenter}.UVs) {parameterName=IconTexture, samplerSource=SSM_Clamp_WorldGroupSettings, texture=/Umbramolt_Hole/UI/Textures/T_UI_UmbramoltKipper_MinimapIcon}
  LinearInterpolate(B=@n1.RGB, Alpha=@n1.RGB)
  LinearInterpolate(
   A=VectorParameter() {parameterName=WaveColor, defaultValue=<rgba 0.9911 0.807 0 1>}.RGB
   B=VectorParameter() {parameterName=PulseColor, defaultValue=<rgba 0.9911 0.807 0 1>}.RGB
   Alpha=NamedRerouteUsage()
  )
  n2 = MaterialFunctionCall(UVs=TextureCoordinate(), "Rotation Angle (0-1)"=Constant() {r=0.0833}) {materialFunction=&CustomRotator}
  MaterialFunctionCall!(
   UVs=@n2.Rotated Values
   Size=Constant() {r=0.5}
   Sides=Constant() {r=6}
   InnerAngle
   OutlineThickness
   OutlineAlignment
   GlowMin=Constant() {r=0.4}
   GlowMax=Constant() {r=0.5}
  ) {materialFunction=/FortUILibrary/MaterialFunctions/Shapes/MF_UI_RegularPolygon}
  n3 = Multiply(A=NamedRerouteUsage()) {constB=0.5}
  n4 = Multiply(A=NamedRerouteUsage()) {constB=0.5}
  Max(A=@n1.A, B=Multiply(A=Max(A=@n3, B=@n4)) {constB=0.6})
  n5 = Frac(None=Multiply(A=Time()) {constB=0.5})
  n6 = MaterialFunctionCall(
   UVs=@n2.Rotated Values
   Size=LinearInterpolate(Alpha=@n5) {constA=0.2, constB=1.2}
   Sides=Constant() {r=6}
   OutlineThickness=LinearInterpolate(Alpha=@n5) {constB=0.03}
  ) {materialFunction=&MF_UI_RegularPolygon}
  n7 = Multiply(A=SmoothStep(Value=OneMinus(None=@n6.SDFOutline)) {constMin=0.99}, B=LinearInterpolate(Alpha=@n5) {constA=1, constB=0})
  n8 = MaterialFunctionCall(UVs=@n2.Rotated Values, Size=Constant() {r=0.7}, Sides=Constant() {r=6}, GlowMin=Constant(), GlowMax=Constant() {r=0.3}) {materialFunction=&MF_UI_RegularPolygon}
  Min(A=@n7, B=@n8.Fill)
  Multiply(A=@n7)
  Multiply(A=@n8.Fill)
  n9 = MaterialFunctionCall(
   UVs=MaterialFunctionCall(UVs=TextureCoordinate(), "Rotation Angle (0-1)"=Constant()) {materialFunction=&CustomRotator}.Rotated Values
   Size=Constant() {r=1}
   Sides=Constant() {r=4}
  ) {materialFunction=&MF_UI_RegularPolygon}
  n10 = Multiply(A=Time(), B=Constant() {desc=speed, r=0.5}) {constB=0.5}
  n11 = Multiply(A=Subtract(A=@n9.SDF, B=Frac(None=@n10)), B=ScalarParameter() {defaultValue=3, desc=repetitions, parameterName=Repetitions})
  n12 = Add(
   A=Multiply(A=Cosine(None=@n11)) {constB=0.5, desc="cosine range is -1 to 1, so we need to bring it back to 0 to 1"}
   B=ScalarParameter() {defaultValue=0.2, parameterName=Thickness}
  ) {constB=0.2}
  n13 = SmoothStep(Value=Add(A=Multiply(A=@n9.SDF) {constB=0.5, desc="cosine range is -1 to 1, so we need to bring it back to 0 to 1"}) {constB=0.2}) {constMax=0.2}
  NamedRerouteDeclaration(None=Multiply(A=SmoothStep(Value=@n12) {constMax=0.6, constMin=0.55}, B=OneMinus(None=@n13))) {name=Wave, nodeColor=<rgba 0 0 0 1>}
  n14 = Time()
  n15 = Multiply(A=Cosine(None=Frac(None=@n14))) {constB=0.1}
  n16 = Add(A=Constant() {r=0.1}, B=@n15)
  n17 = Add(A=@n15, B=Constant() {r=0.2})
  n18 = Constant() {r=0.1}
  n19 = Constant() {r=0.2}
  NamedRerouteDeclaration(None=Add(
    A=MaterialFunctionCall(Size=@n16, GlowMax=@n17) {materialFunction=&MF_UI_Circle}.Fill
    B=MaterialFunctionCall(Size=@n18, GlowMax=@n19) {materialFunction=&MF_UI_Circle}.Fill
   )) {name=Pulse, nodeColor=<rgba 0 0 0 1>}
 ]
}
